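# Sound cue for a Winterfest rubber duck discoverable, layering soft impact, ducky and smoke bomb sounds.
# A Mixer plays three branches: a randomly chosen, pitch-varied bed hit, the ducky wave after a short randomized delay through a Modulator, and a quieter smoke bomb after a fixed delay.
# It uses the UI sound class with overridden attenuation and single-instance concurrency that stops the oldest.
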SoundCue /Game/Athena/Sounds/Events/Winterfest/Discoverables/Winterfest_Discoverable_Duck_Cue
root=Mixer_0
nodes [
 {id=Delay_1, children [Modulator_3], properties {delayMax=0.15, delayMin=0.1}}
 {id=Delay_2, children [WavePlayer_26], properties {delayMax=0.2, delayMin=0.2}}
 {id=Mixer_0, children [Modulator_1,Delay_1,Delay_2], properties {inputVolume [1,0.9,0.3]}}
 {id=Modulator_1, children [Random_9], properties {pitchMax=1.07, pitchMin=0.92}}
 {id=Modulator_3, children [WavePlayer_0]}
 {id=Random_9, children [WavePlayer_48, WavePlayer_49], properties {weights [1,1]}}
 {id=WavePlayer_0, properties {soundWaveAssetPtr=&/Winterfest_Rubber_Ducky_01}}
 {id=WavePlayer_26, properties {soundWaveAssetPtr=/Game/Sounds/Fort_Gadgets/Ninja/smokebomb_02}}
 {id=WavePlayer_48, properties {soundWaveAssetPtr=/Game/Sounds/Fort_Impact_Sounds/Demolition/Bed/Bed_Hit_Soft13}}
 {id=WavePlayer_49, properties {soundWaveAssetPtr=/Game/Sounds/Fort_Impact_Sounds/Demolition/Bed/Bed_Hit_Soft14}}
]
properties {
 bOverrideAttenuation=true
 maxDistance=1200
 soundClassObject=&Fort_UI
 volumeMultiplier=0.6
 attenuationOverrides {falloffDistance=800}
 concurrencyOverrides {maxCount=1, resolutionRule=StopOldest}
}
